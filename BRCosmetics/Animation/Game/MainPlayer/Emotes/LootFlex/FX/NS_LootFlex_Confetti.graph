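# Niagara confetti effect for the LootFlex player emote, though all three of its emitters are disabled.
# Two sprite emitters, confetti and ribbon confetti, fire probabilistic, loop-count-gated bursts in random palette colors, sampling the skeletal mesh and adjusting acceleration for front-end versus gameplay.
# Pieces launch in a cone, then shrink under curve-driven drag, gravity, curl noise and collision, with randomized age-driven material parameters.
# A debug gnomon mesh emitter tracks mesh bones or sockets.
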NiagaraSystem /BRCosmetics/Animation/Game/MainPlayer/Emotes/LootFlex/FX/NS_LootFlex_Confetti
dependencies {
 usedDataInterfaces [&NiagaraDataInterfaceArrayColor, &NiagaraDataInterfaceCurve, &NiagaraDataInterfaceSkeletalMesh]
 usedDynamicInputs []
 usedModules [
  &self:SystemSpawnScript.SystemScriptSource.SystemScriptGraph.NiagaraNodeAssignment_1.SetVariables_#1
  &SystemState
 ]
}
emitters {
 FN_DebugGnomon {
  bEnabled=false
  emitterUpdateScript [
   &EmitterState(
    Life Cycle Mode=::NewEnumerator1
    Loop Behavior=::NewEnumerator1
    Loop Duration Mode=::NewEnumerator1
    Scalability Mode=::NewEnumerator1
    UseLoopDelay=false
   )
   &SpawnBurst_Instantaneous(Spawn Count=4, Use Spawn Probability=false, UseLoopCountLimit=false)
  ]
  particleSpawnScript [
   &InitializeParticle(Color Mode=::NewEnumerator0, Lifetime=5, Mesh Scale Mode=::NewEnumerator3, UsePositionOffset=false, Write Material Random=true)
   &FN_SetSpawnGroup()
  ]
  particleUpdateScript [
   &ParticleState(Kill Particles When Lifetime Has Elapsed=false)
   &SkeletalMeshLocation(
    Bone / Socket Index=$Particles.SpawnGroup
    MaskSpawnGroup=false
    Mesh Sampling Type=::NewEnumerator5
    Skeletal Mesh=$System.Skeletal Mesh
    Socket Sampling Mode=::NewEnumerator5
   ) {module=SkeletalMeshLocation001}
  ]
  rendererClasses [&NiagaraMeshRendererProperties]
  renderers [{rendererClass=&NiagaraMeshRendererProperties, rendererIndex=0}]
 }
 NE_Confetti {
  bEnabled=false
  emitterUpdateScript [
   &EmitterState()
   &SpawnBurst_Instantaneous(
    Spawn Count=&MakeCustomIntFromBool(
     Bool=&SetBoolByIntComparison(A=$Emitter.LoopCount, B=11, Comparison Type=::NewEnumerator2)
     True INT=0
     False INT=&RandomRangeInt(Minimum=4, Maximum=6)
    )
    Spawn Probability=0.25
    Use Spawn Probability=true
    UseLoopCountLimit=false
   )
  ]
  particleSpawnScript [
   &InitializeParticle(
    Color=&SelectLinearColorFromArray(
     Array Sampling Mode=::NewEnumerator1
     Direct Array Index=&RandomRangeInt(Minimum=0, Maximum=6)
     Color Selection Array=DataInterface() {propertyValues {ColorData [
        <rgba 0.7457 0.1697 0.7807 1>
        <rgba 0.4983 0.4184 0.4184 1>
        <rgba 1 0.7353 0.0341 1>
        <rgba 0.6001 0.0194 0.0228 1>
        <rgba 0.0071 0.1214 0.3501 1>
        <rgba 0.0035 0.6649 0.0944 1>
        <rgba 0.2912 0.0938 1 1>
       ]}}
     Random Seed=?
    )
    Lifetime Max=1
    Lifetime Min=2
    Lifetime Mode=::NewEnumerator1
    Position Offset=<0 -22 -8>
    Position Offset Coordinate Space=::Local
    Sprite Size Mode=::NewEnumerator1
    Uniform Sprite Size Max=2
    Uniform Sprite Size Min=4
    UsePositionOffset=true
    Write Material Random=true
   )
   &AccelerationForce(Acceleration=&FN_FrontEndVector3Modifier(Gameplay=&RandomRangeVector(Minimum=<0 0 25>, Maximum=<0 0 100>), Front End=<0 0 10>))
   &SkeletalMeshLocation(MaskSpawnGroup=false, Mesh Sampling Type=::NewEnumerator5, OverrideRandomSeed=false, Skeletal Mesh=$System.Skeletal Mesh)
   &AddVelocity(Velocity=&MakeCustomVectorFromBool(
     Bool=$User.bIsFrontEnd
     True Vector=&MakeVector(
      X=0
      Y=&RandomRangeFloat(Minimum=-40, Maximum=-70)
      Z=&RandomRangeFloat(Minimum=0, Maximum=30)
     )
     False Vector=&MakeVector(
      X=&RandomRangeFloat(Minimum=-150, Maximum=150)
      Y=&RandomRangeFloat(Minimum=-300, Maximum=-100)
      Z=&RandomRangeFloat(Minimum=-80, Maximum=80)
     )
    )) {enabled=false}
   &InheritVelocity(Module.Limit Speed=true, Velocity Limit=30) {enabled=false}
   &ApplyInitialForces()
   &AddVelocity(
    Cone Angle=30
    Module.UseConeFalloff=false
    Module.UseVelDistribution=false
    Override Seed=false
    Velocity Mode=::NewEnumerator2
    Velocity Speed=&RandomRangeFloat(Minimum=30, Maximum=50)
   ) {module=AddVelocity001}
   &ShapeLocation(MaskSpawnGroup=false, Override Seed=false, Shape Origin=$Particles.SkeletalMeshLocation.SampledPosition, Sphere Radius=5, UseOwnerScale=true)
  ]
  particleUpdateScript [
   &ParticleState()
   &ScaleSpriteSize(Uniform Curve Sprite Scale=DataInterface() {propertyValues {Curve {keys [
        {tangentMode=RCTM_User, time=0, value=1}
        {arriveTangent=-0.0405, interpMode=RCIM_Cubic, leaveTangent=-0.0405, tangentMode=RCTM_User, time=0.3823, value=1}
        {arriveTangent=-5.8476, interpMode=RCIM_Cubic, leaveTangent=-5.8476, tangentMode=RCTM_User, time=0.7691, value=0}
       ]}}})
   &Drag(
    Drag=&FloatFromCurve(
     FloatCurve=DataInterface() {propertyValues {Curve {keys [
         {arriveTangent=2.3139, interpMode=RCIM_Cubic, leaveTangent=2.3139, tangentMode=RCTM_User, time=0, value=0}
         {interpMode=RCIM_Cubic, tangentMode=RCTM_User, time=0.4128, value=1}
        ]}}}
     Scale Curve=4
    )
    Use Linear Drag=true
    Use Rotational Drag=true
   )
   &GravityForce(Gravity=<0 0 -344.5266>)
   &CurlNoiseForce(Bias Noise=false, Noise Strength=295.2648, Pan Noise=false)
   &Collision(
    Age Colliding Particles=false
    KillOnCollision=false
    Module.Control Rest State Via Surface Normal=false
    Module.EnableMaxCollisionCount=false
    Module.Randomize Collision Normal Vector=false
   )
   &InheritSourceMovement(Applied Position Scale=<0.6 0.6 0.6>, ApplyForce=false, ApplyPosition=true, ApplyPositionFalloff=false)
   &SolveForcesAndVelocity(Clamp Velocity=false, Limit Acceleration=false)
   &DynamicMaterialParameters(
    Index 0 Param 1=&Add_Float(
     A=&RandomRangeFloat(Minimum=0, Maximum=1000)
     B=&Lerp_Float(A=0, B=&RandomRangeFloat(Minimum=25, Maximum=150), Alpha=$Particles.NormalizedAge)
    )
    Module.Index0Param0Write=true
    Module.Index0Param1Write=true
    Module.Index0Param2Write=true
    Module.Index0Param3Write=true
   )
   &ScaleColor(
    Scale Alpha=&FloatFromCurve(FloatCurve=DataInterface() {propertyValues {Curve {keys [
         {tangentMode=RCTM_User, time=0, value=1}
         {arriveTangent=-0.0405, interpMode=RCIM_Cubic, leaveTangent=-0.0405, tangentMode=RCTM_User, time=0.8203, value=1}
         {arriveTangent=-8.0089, interpMode=RCIM_Cubic, leaveTangent=-8.0089, tangentMode=RCTM_User, time=1, value=0}
        ]}}})
    ScaleA=true
    ScaleRGB=false
   ) {enabled=false}
   &CameraOffset(Camera Offset Amount=50)
  ]
  rendererClasses [&NiagaraSpriteRendererProperties]
  renderers [{rendererClass=&NiagaraSpriteRendererProperties}]
 }
 NE_RibbonConfetti {
  bEnabled=false
  emitterUpdateScript [
   &EmitterState()
   &SpawnBurst_Instantaneous(
    Spawn Count=&MakeCustomIntFromBool(
     Bool=&SetBoolByIntComparison(A=$Emitter.LoopCount, B=11, Comparison Type=::NewEnumerator2)
     True INT=0
     False INT=&RandomRangeInt(Minimum=2, Maximum=5)
    )
    Spawn Probability=0.25
    Use Spawn Probability=true
    UseLoopCountLimit=false
   )
  ]
  particleSpawnScript [
   &InitializeParticle(
    Color=&SelectLinearColorFromArray(
     Array Sampling Mode=::NewEnumerator1
     Direct Array Index=&RandomRangeInt(Minimum=0, Maximum=6)
     Color Selection Array=DataInterface() {propertyValues {ColorData [
        <rgba 0.7807 0.1998 0.5985 1>
        <rgba 0.4983 0.4184 0.4184 1>
        <rgba 0.7251 0.5332 0.0248 1>
        <rgba 0.6001 0.0194 0.0228 1>
        <rgba 0.0071 0.1214 0.3501 1>
        <rgba 0.0035 0.6649 0.0944 1>
        <rgba 0.2912 0.0938 1 1>
       ]}}
     Random Seed=?
    )
    Lifetime Max=0.9
    Lifetime Min=1.5
    Lifetime Mode=::NewEnumerator1
    Position Offset=<0 -22 -8>
    Position Offset Coordinate Space=::Local
    Sprite Size Mode=::NewEnumerator1
    Uniform Sprite Size Max=4.4444
    Uniform Sprite Size Min=5.7778
    UsePositionOffset=true
    Write Material Random=true
   )
   &AccelerationForce(Acceleration=&FN_FrontEndVector3Modifier(Gameplay=&RandomRangeVector(Minimum=<0 0 25>, Maximum=<0 0 100>), Front End=<0 0 10>))
   &SkeletalMeshLocation(MaskSpawnGroup=false, Mesh Sampling Type=::NewEnumerator5, OverrideRandomSeed=false, Skeletal Mesh=$System.Skeletal Mesh)
   &AddVelocity(Velocity=&MakeCustomVectorFromBool(
     Bool=$User.bIsFrontEnd
     True Vector=&MakeVector(
      X=0
      Y=&RandomRangeFloat(Minimum=-40, Maximum=-70)
      Z=&RandomRangeFloat(Minimum=0, Maximum=30)
     )
     False Vector=&MakeVector(
      X=&RandomRangeFloat(Minimum=-150, Maximum=150)
      Y=&RandomRangeFloat(Minimum=-370, Maximum=-100)
      Z=&RandomRangeFloat(Minimum=-80, Maximum=80)
     )
    )) {enabled=false}
   &InheritVelocity(Module.Limit Speed=true, Velocity Limit=30) {enabled=false}
   &ApplyInitialForces()
   &AddVelocity(
    Cone Angle=30
    Module.UseConeFalloff=false
    Module.UseVelDistribution=false
    Override Seed=false
    Velocity Mode=::NewEnumerator2
    Velocity Speed=&RandomRangeFloat(Minimum=30, Maximum=55)
   ) {module=AddVelocity001}
   &ShapeLocation(MaskSpawnGroup=false, Override Seed=false, Sphere Radius=5, UseOwnerScale=true)
  ]
  particleUpdateScript [
   &ParticleState()
   &ScaleSpriteSize(Uniform Curve Sprite Scale=DataInterface() {propertyValues {Curve {keys [
        {tangentMode=RCTM_User, time=0, value=1}
        {arriveTangent=-0.0405, interpMode=RCIM_Cubic, leaveTangent=-0.0405, tangentMode=RCTM_User, time=0.3823, value=1}
        {arriveTangent=-5.8476, interpMode=RCIM_Cubic, leaveTangent=-5.8476, tangentMode=RCTM_User, time=0.7691, value=0}
       ]}}})
   &Drag(
    Drag=&FloatFromCurve(
     FloatCurve=DataInterface() {propertyValues {Curve {keys [
         {arriveTangent=2.7831, interpMode=RCIM_Cubic, leaveTangent=2.7831, tangentMode=RCTM_User, time=0, value=0}
         {interpMode=RCIM_Cubic, tangentMode=RCTM_User, time=0.4035, value=1}
        ]}}}
     Scale Curve=4
    )
    Use Linear Drag=true
    Use Rotational Drag=true
   )
   &GravityForce(Gravity=<0 0 -337.71>)
   &CurlNoiseForce(Bias Noise=false, Noise Strength=295.2648, Pan Noise=false)
   &Collision(
    Age Colliding Particles=false
    KillOnCollision=false
    Module.Control Rest State Via Surface Normal=false
    Module.EnableMaxCollisionCount=false
    Module.Randomize Collision Normal Vector=false
   )
   &InheritSourceMovement(Applied Position Scale=<0.6 0.6 0.6>, ApplyForce=false, ApplyPosition=true, ApplyPositionFalloff=false)
   &SolveForcesAndVelocity(Clamp Velocity=false, Limit Acceleration=false)
   &DynamicMaterialParameters(
    Index 0 Param 1=&Add_Float(
     A=&RandomRangeFloat(Minimum=0, Maximum=1000)
     B=&Lerp_Float(A=0, B=&RandomRangeFloat(Minimum=25, Maximum=150), Alpha=$Particles.NormalizedAge)
    )
    Module.Index0Param0Write=true
    Module.Index0Param1Write=true
    Module.Index0Param2Write=true
    Module.Index0Param3Write=true
   )
   &ScaleColor(
    Scale Alpha=&FloatFromCurve(FloatCurve=DataInterface() {propertyValues {Curve {keys [
         {tangentMode=RCTM_User, time=0, value=1}
         {arriveTangent=-0.0405, interpMode=RCIM_Cubic, leaveTangent=-0.0405, tangentMode=RCTM_User, time=0.8203, value=1}
         {arriveTangent=-8.0089, interpMode=RCIM_Cubic, leaveTangent=-8.0089, tangentMode=RCTM_User, time=1, value=0}
        ]}}})
    ScaleA=true
    ScaleRGB=false
   ) {enabled=false}
   &CameraOffset(Camera Offset Amount=50)
  ]
  rendererClasses [&NiagaraSpriteRendererProperties]
  renderers [{rendererClass=&NiagaraSpriteRendererProperties}]
 }
}
summary {
 systemName=NS_LootFlex_Confetti
 emitters [
  {bEnabled=false, emitterName=FN_DebugGnomon, rendererClasses [&NiagaraMeshRendererProperties]}
  {bEnabled=false, emitterName=NE_Confetti, rendererClasses [&NiagaraSpriteRendererProperties]}
  {bEnabled=false, emitterName=NE_RibbonConfetti, rendererClasses [&NiagaraSpriteRendererProperties]}
 ]
 userVariables [
  User.bIsFrontEnd {type=NiagaraBool}
  User.bIsFrontEndPreview {type=NiagaraBool}
 ]
}
