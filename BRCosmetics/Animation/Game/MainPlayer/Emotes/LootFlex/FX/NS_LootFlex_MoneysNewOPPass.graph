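# Niagara effect for the LootFlex emote that throws coins and multicolored V-Bucks meshes from character skeletal mesh sockets.
# Coin and VBucks emitters combine loop-count-gated bursts with movement-based spawning, while their 001 variants fire larger bursts.
# Spawn groups pick each particle's socket and cone direction, then gravity, drag and quality-gated collision shape the motion.
# Coins shimmer via sine-driven sprites with time-of-day scaling; V-Bucks flutter via curl noise and aerodynamic drag.
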NiagaraSystem /BRCosmetics/Animation/Game/MainPlayer/Emotes/LootFlex/FX/NS_LootFlex_MoneysNewOPPass
dependencies {
 usedDataInterfaces [
  &NiagaraDataInterfaceArrayColor
  &NiagaraDataInterfaceArrayFloat3
  &NiagaraDataInterfaceCurve
  &NiagaraDataInterfacePlatformSet
  &NiagaraDataInterfaceArrayInt32
  &NiagaraDataInterfaceSpriteRendererInfo
  &NiagaraDataInterfaceSkeletalMesh
 ]
 usedDynamicInputs []
 usedModules [
  &self:SystemSpawnScript.SystemScriptSource.SystemScriptGraph.NiagaraNodeAssignment_0.SetVariables_#1
  &SystemState
 ]
}
emitters {
 Coin {
  emitterUpdateScript [
   &EmitterState()
   &SpawnBurst_Instantaneous(
    Spawn Count=&MakeCustomIntFromBool(
     Bool=&SetBoolByIntComparison(A=$Emitter.LoopCount, B=13, Comparison Type=::NewEnumerator2)
     True INT=0
     False INT=10
    )
    Use Spawn Probability=false
    UseLoopCountLimit=false
   )
   &SpawnRate(SpawnRate=20, Use Spawn Probability=false) {enabled=false}
   &SpawnRate(Spawn Probability=0.5, SpawnRate=5, Use Spawn Probability=true) {enabled=false, module=SpawnRate001}
   &SpawnPerUnit(
    Module.Use Max Movement Threshold=true
    Module.Use Movement Tolerance=true
    Movement Tolerance=50
    Spawn Spacing=7
    Use Spawn Probability=false
   )
  ]
  particleSpawnScript [
   /SetVariables(Particles.VisibilityTag=&SelectIntFromArray(
     Direct Array Index=&RandomRangeInt(Minimum=0, Maximum=2)
     Int Selection Array=DataInterface() {propertyValues {IntData [0,1,1]}}
    )) {isSetParameters=true, module=SetVariables_#2}
   &InitializeParticle(
    Lifetime Max=0.6
    Lifetime Min=1.5
    Lifetime Mode=::NewEnumerator1
    Mesh Scale Mode=::NewEnumerator3
    Mesh Uniform Scale=&RandomRangeFloat(Minimum=0.05, Maximum=0.03)
    Sprite Rotation Mode=::NewEnumerator2
    Sprite Size Mode=::NewEnumerator3
    Uniform Sprite Size=5
    UsePositionOffset=false
    Write Material Random=true
   )
   &FN_SetSpawnGroup()
   &SkeletalMeshLocation(MaskSpawnGroup=false, Mesh Sampling Type=::NewEnumerator5, OverrideRandomSeed=false, Skeletal Mesh=$System.Skeletal Mesh) {enabled=false, module=SkeletalMeshLocation001}
   &SkeletalMeshLocation(Bone Sampling Mode=::NewEnumerator4, MaskSpawnGroup=false, Sampled Position Offset=<0 -25 0>, Skeletal Mesh=$System.VehicleLocation) {enabled=false}
   &SkeletalMeshLocation(
    Bone / Socket Index=$Particles.SpawnGroup
    MaskSpawnGroup=false
    Mesh Sampling Type=::NewEnumerator5
    Skeletal Mesh=$System.Skeletal Mesh
    Socket Sampling Mode=::NewEnumerator5
   ) {module=SkeletalMeshLocation002}
   &ShapeLocation(
    MaskSpawnGroup=false
    Override Seed=false
    Shape Origin=$Particles.SkeletalMeshLocation002.SampledPosition
    Sphere Radius=2
    UseOwnerScale=true
   )
   &InitialMeshOrientation(Use Rotation Vector=false)
   /SetVariables(Particles.SpriteColor=<rgba 1 0.9843 0.9961 1>) {isSetParameters=true, module=SetVariables_#3}
   /SetVariables(Particles.MeshColor:LinearColor=<rgba 1 0.5686 0.2157 0>) {isSetParameters=true}
   &AddVelocity(
    Rotation Coordinate Space=::World
    Velocity=&RandomRangeVector(Minimum=<-800 -800 300>, Maximum=<800 800 1000>)
    Velocity Speed Scale=0.2
   ) {enabled=false}
   &AddVelocity(
    Cone Angle=55
    Cone Axis=&SelectVectorFromArray(
     Direct Array Index=$Particles.SpawnGroup
     Vector Selection Array=DataInterface() {propertyValues {FloatData [<1 0 0>, <-1 -1 0>, <0 0 0>, <-0.5 1 0>]}}
    )
    Module.UseConeFalloff=false
    Module.UseVelDistribution=false
    Override Seed=false
    Velocity Mode=::NewEnumerator2
    Velocity Speed=&RandomRangeFloat(Minimum=120, Maximum=40)
   ) {module=AddVelocity001}
   &AddVelocity(Rotation Coordinate Space=::World, Velocity=&RandomRangeVector(Minimum=<0 0 40>, Maximum=<0 0 200>)) {module=AddVelocity002}
  ]
  particleUpdateScript [
   /SetVariables(Particles.SineWave:NiagaraFloat=&Sine(Normalized Angle=$Emitter.Age, Period=&RandomRangeFloat(Minimum=0.3, Maximum=0.7), Scale=1, Bias=0)) {isSetParameters=true, module=SetVariables_#4}
   &ParticleState()
   &ScaleSpriteSize(Scale Factor=&MakeVector2D(X=$Particles.SineWave, Y=1), Scale Sprite Size Mode=::NewEnumerator2) {module=ScaleSpriteSize001}
   &SpriteRotationRate(Rotation Rate=&RandomRangeFloat(Minimum=-1000, Maximum=-500))
   &ScaleMeshSize(Scale Factor=&VectorFromFloat(Value=&FloatFromCurve(FloatCurve=DataInterface() {propertyValues {Curve {keys [<at 0.6 1>,<at 1 0>]}}})))
   &ScaleSpriteSize(Uniform Curve Sprite Scale=DataInterface() {propertyValues {Curve {keys [<at 0.001 0.5>,<at 0.2948 1.1652>,<at 1.0009 0.8072>]}}})
   &ScaleColor(
    Scale Alpha=&FloatFromCurve(
     FloatCurve=DataInterface() {propertyValues {Curve {keys [
         {arriveTangent=-0.0888, interpMode=RCIM_Cubic, leaveTangent=-0.0888, tangentMode=RCTM_User, time=0, value=0}
         {arriveTangent=0.12, interpMode=RCIM_Cubic, leaveTangent=0.12, tangentMode=RCTM_User, time=0.15, value=0.9847}
         {arriveTangent=-0.0405, interpMode=RCIM_Cubic, leaveTangent=-0.0405, tangentMode=RCTM_User, time=1, value=0}
        ]}}}
     Scale Curve=&FNTimeOfDayScalar(
      Night Scale=1
      Morning Scale=1
      Day Scale=&FN_FrontEndFloatModifier(Gameplay=1, Front End=1)
      Evening Scale=1
     )
    )
    Scale RGB=&VectorFromFloat(Value=&FNTimeOfDayScalar(
      Night Scale=0.8
      Morning Scale=0.8
      Day Scale=&FN_FrontEndFloatModifier(Gameplay=1, Front End=1)
      Evening Scale=0.5
     ))
    ScaleA=true
    ScaleRGB=true
   ) {module=ScaleColor003}
   &Drag(
    Drag=&FloatFromCurve(
     FloatCurve=DataInterface() {propertyValues {Curve {keys [
         {arriveTangent=0.7001, interpMode=RCIM_Cubic, leaveTangent=0.7001, tangentMode=RCTM_User, time=0, value=0}
         {interpMode=RCIM_Cubic, tangentMode=RCTM_User, time=1, value=1}
        ]}}}
     Scale Curve=4
    )
    Use Linear Drag=true
    Use Rotational Drag=true
   ) {module=Drag002}
   &GravityForce(Gravity=&RandomRangeVector(Minimum=<0 0 -980>, Maximum=<0 0 -600>))
   &UpdateMeshOrientation(Rotation Vector=&Multiply_Vector(
     A=<1 1 1>
     B=&MakeVector(
      X=&MakeCustomFloatFromBool(Bool=&RandomBool(), True Float=2, False Float=-2)
      Y=&RandomRangeFloat(Minimum=0.1, Maximum=-0.1)
      Z=&RandomRangeFloat(Minimum=0.1, Maximum=-0.1)
     )
    ))
   &Collision(
    Age Colliding Particles=false
    Collision Enabled=&Boolean_AndOperation(
     Bool1=&IsPlatformSetActive(New Platform Set=DataInterface() {propertyValues {Platforms {qualityLevelMask=-8}}})
     Bool2=&SetBoolByFloatComparison(A=$Emitter.Age, B=0.4)
    )
    KillOnCollision=false
    Module.Control Rest State Via Surface Normal=false
    Module.EnableMaxCollisionCount=false
    Module.Randomize Collision Normal Vector=false
    Radius Calculation Type=::NewEnumerator1
   )
   &Collision(
    Age Colliding Particles=false
    Analytical Collision Plane Position 1:Vector3f=<0 0 -20>
    Analytical Collision Plane Space:ENiagaraCoordinateSpace=::Local
    CPU Collision Type=::NewEnumerator3
    Collision Enabled=&IsPlatformSetActive(New Platform Set=DataInterface() {propertyValues {Platforms {qualityLevelMask=-28}}})
    KillOnCollision=false
    Module.Control Rest State Via Surface Normal=false
    Module.EnableMaxCollisionCount=false
    Module.Randomize Collision Normal Vector=false
    Perform 2 Analytical Collision Queries=false
   ) {enabled=false, module=Collision001}
   &SolveForcesAndVelocity(Clamp Velocity=false, Limit Acceleration=false)
   &DynamicMaterialParameters(
    Index 0 Param 1=&Add_Float(
     A=&RandomRangeFloat(Minimum=0, Maximum=1)
     B=&FloatFromCurve(
      FloatCurve=DataInterface() {propertyValues {Curve {keys [
          {arriveTangent=2.7533, interpMode=RCIM_Cubic, leaveTangent=1.8958, tangentMode=RCTM_User, time=0, value=1}
          {arriveTangent=0.8292, interpMode=RCIM_Cubic, leaveTangent=0.8292, tangentMode=RCTM_User, time=0.321, value=1.3453}
          {arriveTangent=0.0529, interpMode=RCIM_Cubic, leaveTangent=0.2021, tangentMode=RCTM_User, time=1, value=1.5}
         ]}}}
      CurveIndex=&Add_Float(A=&Multiply_Float(A=1, B=0.5), B=$Particles.NormalizedAge)
     )
    )
    Index 0 Param 2=&FloatFromCurve(FloatCurve=DataInterface() {propertyValues {Curve {keys [
         {arriveTangent=0.0453, interpMode=RCIM_Cubic, tangentMode=RCTM_Break, time=0, value=-1}
         {arriveTangent=0.0604, interpMode=RCIM_Cubic, leaveTangent=0.0604, tangentMode=RCTM_User, time=1, value=-1}
        ]}}})
    Index 0 Param 3=&FloatFromCurve(FloatCurve=DataInterface() {propertyValues {Curve {keys [
         {arriveTangent=-53.9199, interpMode=RCIM_Cubic, leaveTangent=-8.9727, tangentMode=RCTM_User, time=0, value=1}
         {arriveTangent=0.2347, interpMode=RCIM_Cubic, leaveTangent=0.2723, tangentMode=RCTM_Break, time=0.25, value=0}
        ]}}})
    Index 0 Param 4=&FloatFromCurve(
     FloatCurve=DataInterface() {propertyValues {Curve {keys [<at 0.5 1>,<at 1 0>]}}}
     Scale Curve=&FNTimeOfDayScalar(
      Night Scale=2
      Morning Scale=2
      Day Scale=&FN_FrontEndFloatModifier(Gameplay=1, Front End=0.5)
      Evening Scale=3
     )
    )
    Index 2 Param 1=&FloatFromCurve(FloatCurve=DataInterface() {propertyValues {Curve {keys [
         {interpMode=RCIM_Cubic, leaveTangent=4.2485, tangentMode=RCTM_Break, time=0, value=0}
         {arriveTangent=1.0548, interpMode=RCIM_Cubic, leaveTangent=1.0548, tangentMode=RCTM_User, time=0.17, value=0.5715}
         <at 1 1>
        ]}}})
    Index 2 Param 2=&FloatFromCurve(
     FloatCurve=DataInterface() {propertyValues {Curve {keys [
         {arriveTangent=1.9056, interpMode=RCIM_Cubic, leaveTangent=1.9056, tangentMode=RCTM_User, time=0, value=0.2}
         {arriveTangent=0.0259, interpMode=RCIM_Cubic, leaveTangent=0.0259, tangentMode=RCTM_User, time=1, value=1.4}
        ]}}}
     Scale Curve=&RandomRangeFloat(Minimum=0.2, Maximum=0.5)
    )
    Index 2 Param 3=&FloatFromCurve(FloatCurve=DataInterface() {propertyValues {Curve {keys [
         {interpMode=RCIM_Cubic, leaveTangent=2.3064, tangentMode=RCTM_Break, time=0, value=1}
         {interpMode=RCIM_Cubic, leaveTangent=11.1233, tangentMode=RCTM_Break, time=1, value=2}
        ]}}})
    Index 2 Param 4=&FloatFromCurve(
     FloatCurve=DataInterface() {propertyValues {Curve {keys [
         {arriveTangent=-0.6867, interpMode=RCIM_Cubic, leaveTangent=-0.6867, tangentMode=RCTM_User, time=0, value=1}
         {interpMode=RCIM_Cubic, leaveTangent=-1.2956, tangentMode=RCTM_Break, time=1, value=0.75}
        ]}}}
     Scale Curve=1.6
    )
    Module.Param0WriteEnabled=true
    Module.Param0WriteEnabled2=true
    Module.Param1WriteEnabled=false
    Module.Param1WriteEnabled2=true
    Module.Param2WriteEnabled=false
    Module.Param2WriteEnabled2=true
    Module.Param3WriteEnabled=true
    Module.Param3WriteEnabled2=true
    Write Parameter Index 2=true
   )
   &SubUVAnimation(
    Evaluation Type=::NewEnumerator0
    "Sprite Renderer "=DataInterface() {propertyValues {SpriteRenderer=&self:Coin_0.NiagaraSpriteRendererProperties_0}}
    SubUV Animation Mode:ENiagara_SubUVLookupModeV2=::NewEnumerator2
    UseEndFrame=false
    UseStartFrame=false
   )
   &ScaleColor(
    Scale RGB=&VectorFromFloat(Value=&RemapRange(Input Value To Remap=$Particles.SineWave, Input Min=-1, Input Max=1, Output Min=1, Output Max=3))
    ScaleA=true
    ScaleRGB=true
   )
  ]
  rendererClasses [&NiagaraMeshRendererProperties, &NiagaraSpriteRendererProperties]
  renderers [
   {rendererClass=&NiagaraMeshRendererProperties, rendererIndex=0}
   {rendererClass=&NiagaraSpriteRendererProperties, rendererIndex=1}
  ]
 }
 Coin001 {
  emitterUpdateScript [
   &EmitterState(Life Cycle Mode=::NewEnumerator1, Loop Behavior=::NewEnumerator1, Loop Delay=0.62, Loop Duration=2, UseLoopDelay=false)
   &SpawnBurst_Instantaneous(Spawn Count=25, Use Spawn Probability=false, UseLoopCountLimit=false)
   &SpawnRate(SpawnRate=20, Use Spawn Probability=false) {enabled=false}
   &SpawnRate(Spawn Probability=0.5, SpawnRate=5, Use Spawn Probability=true) {enabled=false, module=SpawnRate001}
   &SpawnPerUnit(Module.Use Max Movement Threshold=true, Module.Use Movement Tolerance=false, Spawn Spacing=7, Use Spawn Probability=false) {enabled=false}
  ]
  particleSpawnScript [
   /SetVariables(Particles.VisibilityTag=&SelectIntFromArray(
     Direct Array Index=&RandomRangeInt(Minimum=0, Maximum=2)
     Int Selection Array=DataInterface() {propertyValues {IntData [0,1,1]}}
    )) {isSetParameters=true, module=SetVariables_#2}
   &InitializeParticle(
    Lifetime Max=0.6
    Lifetime Min=1.5
    Lifetime Mode=::NewEnumerator1
    Mesh Scale Mode=::NewEnumerator3
    Mesh Uniform Scale=&RandomRangeFloat(Minimum=0.05, Maximum=0.03)
    Sprite Rotation Mode=::NewEnumerator2
    Sprite Size Mode=::NewEnumerator3
    Uniform Sprite Size=5
    UsePositionOffset=false
    Write Material Random=true
   )
   &FN_SetSpawnGroup()
   &SkeletalMeshLocation(MaskSpawnGroup=false, Mesh Sampling Type=::NewEnumerator5, OverrideRandomSeed=false, Skeletal Mesh=$System.Skeletal Mesh) {enabled=false, module=SkeletalMeshLocation001}
   &SkeletalMeshLocation(Bone Sampling Mode=::NewEnumerator4, MaskSpawnGroup=false, Sampled Position Offset=<0 -25 0>, Skeletal Mesh=$System.VehicleLocation) {enabled=false}
   &SkeletalMeshLocation(
    Bone / Socket Index=$Particles.SpawnGroup
    MaskSpawnGroup=false
    Mesh Sampling Type=::NewEnumerator5
    Skeletal Mesh=$System.Skeletal Mesh
    Socket Sampling Mode=::NewEnumerator5
   ) {module=SkeletalMeshLocation002}
   &ShapeLocation(
    MaskSpawnGroup=false
    Override Seed=false
    Shape Origin=$Particles.SkeletalMeshLocation002.SampledPosition
    Sphere Radius=2
    UseOwnerScale=true
   )
   &InitialMeshOrientation(Use Rotation Vector=false)
   /SetVariables(Particles.SpriteColor=<rgba 1 0.9843 0.9961 1>) {isSetParameters=true, module=SetVariables_#3}
   /SetVariables(Particles.MeshColor:LinearColor=<rgba 1 0.5686 0.2157 0>) {isSetParameters=true}
   &AddVelocity(
    Rotation Coordinate Space=::World
    Velocity=&RandomRangeVector(Minimum=<-800 -800 300>, Maximum=<800 800 1000>)
    Velocity Speed Scale=0.2
   ) {enabled=false}
   &AddVelocity(
    Cone Angle=55
    Cone Axis=&SelectVectorFromArray(
     Direct Array Index=$Particles.SpawnGroup
     Vector Selection Array=DataInterface() {propertyValues {FloatData [<1 0 0>, <-1 -1 0>, <0 0 0>, <-0.5 1 0>]}}
    )
    Module.UseConeFalloff=false
    Module.UseVelDistribution=false
    Override Seed=false
    Velocity Mode=::NewEnumerator2
    Velocity Speed=&RandomRangeFloat(Minimum=120, Maximum=40)
   ) {module=AddVelocity001}
   &AddVelocity(Rotation Coordinate Space=::World, Velocity=&RandomRangeVector(Minimum=<0 0 40>, Maximum=<0 0 200>)) {module=AddVelocity002}
  ]
  particleUpdateScript [
   /SetVariables(Particles.SineWave:NiagaraFloat=&Sine(Normalized Angle=$Emitter.Age, Period=&RandomRangeFloat(Minimum=0.3, Maximum=0.7), Scale=1, Bias=0)) {isSetParameters=true, module=SetVariables_#4}
   &ParticleState()
   &ScaleSpriteSize(Scale Factor=&MakeVector2D(X=$Particles.SineWave, Y=1), Scale Sprite Size Mode=::NewEnumerator2) {module=ScaleSpriteSize001}
   &SpriteRotationRate(Rotation Rate=&RandomRangeFloat(Minimum=-1000, Maximum=-500))
   &ScaleMeshSize(Scale Factor=&VectorFromFloat(Value=&FloatFromCurve(FloatCurve=DataInterface() {propertyValues {Curve {keys [<at 0.6 1>,<at 1 0>]}}})))
   &ScaleSpriteSize(Uniform Curve Sprite Scale=DataInterface() {propertyValues {Curve {keys [<at 0.001 0.5>,<at 0.2948 1.1652>,<at 1.0009 0.8072>]}}})
   &ScaleColor(
    Scale Alpha=&FloatFromCurve(
     FloatCurve=DataInterface() {propertyValues {Curve {keys [
         {arriveTangent=-0.0888, interpMode=RCIM_Cubic, leaveTangent=-0.0888, tangentMode=RCTM_User, time=0, value=0}
         {arriveTangent=0.12, interpMode=RCIM_Cubic, leaveTangent=0.12, tangentMode=RCTM_User, time=0.15, value=0.9847}
         {arriveTangent=-0.0405, interpMode=RCIM_Cubic, leaveTangent=-0.0405, tangentMode=RCTM_User, time=1, value=0}
        ]}}}
     Scale Curve=&FNTimeOfDayScalar(
      Night Scale=1
      Morning Scale=1
      Day Scale=&FN_FrontEndFloatModifier(Gameplay=1, Front End=1)
      Evening Scale=1
     )
    )
    Scale RGB=&VectorFromFloat(Value=&FNTimeOfDayScalar(
      Night Scale=0.8
      Morning Scale=0.8
      Day Scale=&FN_FrontEndFloatModifier(Gameplay=1, Front End=1)
      Evening Scale=0.5
     ))
    ScaleA=true
    ScaleRGB=true
   ) {module=ScaleColor003}
   &Drag(
    Drag=&FloatFromCurve(
     FloatCurve=DataInterface() {propertyValues {Curve {keys [
         {arriveTangent=0.7001, interpMode=RCIM_Cubic, leaveTangent=0.7001, tangentMode=RCTM_User, time=0, value=0}
         {interpMode=RCIM_Cubic, tangentMode=RCTM_User, time=1, value=1}
        ]}}}
     Scale Curve=4
    )
    Use Linear Drag=true
    Use Rotational Drag=true
   ) {module=Drag002}
   &GravityForce(Gravity=&RandomRangeVector(Minimum=<0 0 -980>, Maximum=<0 0 -600>))
   &UpdateMeshOrientation(Rotation Vector=&Multiply_Vector(
     A=<1 1 1>
     B=&MakeVector(
      X=&MakeCustomFloatFromBool(Bool=&RandomBool(), True Float=2, False Float=-2)
      Y=&RandomRangeFloat(Minimum=0.1, Maximum=-0.1)
      Z=&RandomRangeFloat(Minimum=0.1, Maximum=-0.1)
     )
    ))
   &Collision(
    Age Colliding Particles=false
    Collision Enabled=&Boolean_AndOperation(
     Bool1=&IsPlatformSetActive(New Platform Set=DataInterface() {propertyValues {Platforms {qualityLevelMask=-8}}})
     Bool2=&SetBoolByFloatComparison(A=$Emitter.Age, B=0.4)
    )
    KillOnCollision=false
    Module.Control Rest State Via Surface Normal=false
    Module.EnableMaxCollisionCount=false
    Module.Randomize Collision Normal Vector=false
    Radius Calculation Type=::NewEnumerator1
   )
   &Collision(
    Age Colliding Particles=false
    Analytical Collision Plane Position 1:Vector3f=<0 0 -20>
    Analytical Collision Plane Space:ENiagaraCoordinateSpace=::Local
    CPU Collision Type=::NewEnumerator3
    Collision Enabled=&IsPlatformSetActive(New Platform Set=DataInterface() {propertyValues {Platforms {qualityLevelMask=-28}}})
    KillOnCollision=false
    Module.Control Rest State Via Surface Normal=false
    Module.EnableMaxCollisionCount=false
    Module.Randomize Collision Normal Vector=false
    Perform 2 Analytical Collision Queries=false
   ) {module=Collision001}
   &SolveForcesAndVelocity(Clamp Velocity=false, Limit Acceleration=false)
   &DynamicMaterialParameters(
    Index 0 Param 1=&Add_Float(
     A=&RandomRangeFloat(Minimum=0, Maximum=1)
     B=&FloatFromCurve(
      FloatCurve=DataInterface() {propertyValues {Curve {keys [
          {arriveTangent=2.7533, interpMode=RCIM_Cubic, leaveTangent=1.8958, tangentMode=RCTM_User, time=0, value=1}
          {arriveTangent=0.8292, interpMode=RCIM_Cubic, leaveTangent=0.8292, tangentMode=RCTM_User, time=0.321, value=1.3453}
          {arriveTangent=0.0529, interpMode=RCIM_Cubic, leaveTangent=0.2021, tangentMode=RCTM_User, time=1, value=1.5}
         ]}}}
      CurveIndex=&Add_Float(A=&Multiply_Float(A=1, B=0.5), B=$Particles.NormalizedAge)
     )
    )
    Index 0 Param 2=&FloatFromCurve(FloatCurve=DataInterface() {propertyValues {Curve {keys [
         {arriveTangent=0.0453, interpMode=RCIM_Cubic, tangentMode=RCTM_Break, time=0, value=-1}
         {arriveTangent=0.0604, interpMode=RCIM_Cubic, leaveTangent=0.0604, tangentMode=RCTM_User, time=1, value=-1}
        ]}}})
    Index 0 Param 3=&FloatFromCurve(FloatCurve=DataInterface() {propertyValues {Curve {keys [
         {arriveTangent=-53.9199, interpMode=RCIM_Cubic, leaveTangent=-8.9727, tangentMode=RCTM_User, time=0, value=1}
         {arriveTangent=0.2347, interpMode=RCIM_Cubic, leaveTangent=0.2723, tangentMode=RCTM_Break, time=0.25, value=0}
        ]}}})
    Index 0 Param 4=&FloatFromCurve(
     FloatCurve=DataInterface() {propertyValues {Curve {keys [<at 0.5 1>,<at 1 0>]}}}
     Scale Curve=&FNTimeOfDayScalar(
      Night Scale=2
      Morning Scale=2
      Day Scale=&FN_FrontEndFloatModifier(Gameplay=1, Front End=0.5)
      Evening Scale=3
     )
    )
    Index 2 Param 1=&FloatFromCurve(FloatCurve=DataInterface() {propertyValues {Curve {keys [
         {interpMode=RCIM_Cubic, leaveTangent=4.2485, tangentMode=RCTM_Break, time=0, value=0}
         {arriveTangent=1.0548, interpMode=RCIM_Cubic, leaveTangent=1.0548, tangentMode=RCTM_User, time=0.17, value=0.5715}
         <at 1 1>
        ]}}})
    Index 2 Param 2=&FloatFromCurve(
     FloatCurve=DataInterface() {propertyValues {Curve {keys [
         {arriveTangent=1.9056, interpMode=RCIM_Cubic, leaveTangent=1.9056, tangentMode=RCTM_User, time=0, value=0.2}
         {arriveTangent=0.0259, interpMode=RCIM_Cubic, leaveTangent=0.0259, tangentMode=RCTM_User, time=1, value=1.4}
        ]}}}
     Scale Curve=&RandomRangeFloat(Minimum=0.2, Maximum=0.5)
    )
    Index 2 Param 3=&FloatFromCurve(FloatCurve=DataInterface() {propertyValues {Curve {keys [
         {interpMode=RCIM_Cubic, leaveTangent=2.3064, tangentMode=RCTM_Break, time=0, value=1}
         {interpMode=RCIM_Cubic, leaveTangent=11.1233, tangentMode=RCTM_Break, time=1, value=2}
        ]}}})
    Index 2 Param 4=&FloatFromCurve(
     FloatCurve=DataInterface() {propertyValues {Curve {keys [
         {arriveTangent=-0.6867, interpMode=RCIM_Cubic, leaveTangent=-0.6867, tangentMode=RCTM_User, time=0, value=1}
         {interpMode=RCIM_Cubic, leaveTangent=-1.2956, tangentMode=RCTM_Break, time=1, value=0.75}
        ]}}}
     Scale Curve=1.6
    )
    Module.Param0WriteEnabled=true
    Module.Param0WriteEnabled2=true
    Module.Param1WriteEnabled=false
    Module.Param1WriteEnabled2=true
    Module.Param2WriteEnabled=false
    Module.Param2WriteEnabled2=true
    Module.Param3WriteEnabled=true
    Module.Param3WriteEnabled2=true
    Write Parameter Index 2=true
   )
   &SubUVAnimation(
    Evaluation Type=::NewEnumerator0
    "Sprite Renderer "=DataInterface() {propertyValues {SpriteRenderer=&self:Coin001_1.NiagaraSpriteRendererProperties_0}}
    SubUV Animation Mode:ENiagara_SubUVLookupModeV2=::NewEnumerator2
    UseEndFrame=false
    UseStartFrame=false
   )
   &ScaleColor(
    Scale RGB=&VectorFromFloat(Value=&RemapRange(Input Value To Remap=$Particles.SineWave, Input Min=-1, Input Max=1, Output Min=1, Output Max=3))
    ScaleA=true
    ScaleRGB=true
   )
  ]
  rendererClasses [&NiagaraSpriteRendererProperties, &NiagaraMeshRendererProperties]
  renderers [
   {rendererClass=&NiagaraSpriteRendererProperties}
   {rendererClass=&NiagaraMeshRendererProperties, rendererIndex=1}
  ]
 }
 VBucks {
  emitterUpdateScript [
   &EmitterState()
   &SpawnBurst_Instantaneous(
    Spawn Count=&MakeCustomIntFromBool(
     Bool=&SetBoolByIntComparison(A=$Emitter.LoopCount, B=11, Comparison Type=::NewEnumerator2)
     True INT=0
     False INT=4
    )
    Spawn Probability=0.75
    Use Spawn Probability=true
    UseLoopCountLimit=false
   )
   &SpawnRate(SpawnRate=5, Use Spawn Probability=false) {enabled=false, module=SpawnRate001}
   &SpawnRate(SpawnRate=10, Use Spawn Probability=false) {enabled=false}
   &SpawnPerUnit(
    Module.Use Max Movement Threshold=true
    Module.Use Movement Tolerance=true
    Movement Tolerance=50
    Spawn Probability=0.6
    Spawn Spacing:NiagaraFloat=&RandomRangeFloat(Minimum=10, Maximum=20)
    Use Spawn Probability=true
   )
  ]
  particleSpawnScript [
   &InitializeParticle(
    Color=&SelectLinearColorFromArray(
     Array Sampling Mode=::NewEnumerator0
     Direct Array Index=?
     Color Selection Array=DataInterface() {propertyValues {ColorData [<rgba 0.9 0.81 0.225 1>, <rgba 1 0.45 0.5875 1>, <rgba 0.36 0.9 0.5535 1>, <rgba 0.8019 0.5 1 1>, <rgba 1 0.5127 0.3 1>]}}
     Random Seed=0
    )
    Lifetime Max=2.5
    Lifetime Min=1.4
    Lifetime Mode=::NewEnumerator1
    Mesh Scale Mode=::NewEnumerator3
    Mesh Uniform Scale=0.85
    UsePositionOffset=false
    Write Material Random=true
   )
   &FN_SetSpawnGroup()
   &SkeletalMeshLocation(
    Bone / Socket Index=$Particles.SpawnGroup
    MaskSpawnGroup=false
    Mesh Sampling Type=::NewEnumerator5
    Skeletal Mesh=$System.Skeletal Mesh
    Socket Sampling Mode=::NewEnumerator5
   )
   &ShapeLocation(MaskSpawnGroup=false, Override Seed=false, Shape Origin=$Particles.SkeletalMeshLocation.SampledPosition, Sphere Radius=5, UseOwnerScale=true)
   &InitialMeshOrientation(Use Rotation Vector=false)
   &AddVelocity(Velocity=&RandomRangeVector(Minimum=<-300 -400 0>, Maximum=<300 400 50>)) {enabled=false}
   &AddVelocity(
    Cone Angle=15
    Cone Axis=&SelectVectorFromArray(
     Direct Array Index=$Particles.SpawnGroup
     Vector Selection Array=DataInterface() {propertyValues {FloatData [<1 0 0>, <-1 -1 0>, <0 0 0>, <-0.5 1 0>]}}
    )
    Module.UseConeFalloff=false
    Module.UseVelDistribution=false
    Override Seed=false
    Velocity Mode=::NewEnumerator2
    Velocity Speed=&RandomRangeFloat(Minimum=100, Maximum=225)
   ) {module=AddVelocity001}
   &AddVelocity(Rotation Coordinate Space=::World, Velocity=&RandomRangeVector(Minimum=<0 0 40>, Maximum=<0 0 120>)) {module=AddVelocity002}
  ]
  particleUpdateScript [
   &ParticleState()
   &ScaleMeshSize(Scale Factor=&VectorFromFloat(Value=&FloatFromCurve(
      FloatCurve=DataInterface() {propertyValues {Curve {keys [
          {arriveTangent=4.1785, interpMode=RCIM_Cubic, leaveTangent=8.7018, tangentMode=RCTM_Break, time=0, value=0}
          {arriveTangent=0.0224, interpMode=RCIM_Cubic, leaveTangent=0.0224, tangentMode=RCTM_Break, time=0.3, value=1}
         ]}}}
      CurveIndex=$Particles.Age
      Scale Curve=&FloatFromCurve(FloatCurve=DataInterface() {propertyValues {Curve {keys [
           {interpMode=RCIM_Cubic, tangentMode=RCTM_Break, time=0.7176, value=1}
           {arriveTangent=-9.1079, interpMode=RCIM_Cubic, tangentMode=RCTM_Break, time=1, value=0}
          ]}}})
     )))
   &GravityForce()
   /Niagara/Modules/Update/Forces/CurlNoiseForce(Noise Frequency=4.9092, Noise Strength=1143.17, Pan Noise=false) {module=CurlNoiseForce001}
   &AerodynamicDrag(
    Rotation Strength:NiagaraFloat=&RandomRangeFloat(Minimum=2.4141, Maximum=5.2481)
    Use Linear Drag=true
    Use Rotational Drag=true
   )
   &Collision(
    Age Colliding Particles=false
    Collision Enabled=&IsPlatformSetActive(New Platform Set=DataInterface() {propertyValues {Platforms {qualityLevelMask=-8}}})
    KillOnCollision=false
    Module.Control Rest State Via Surface Normal=false
    Module.EnableMaxCollisionCount=false
    Module.Randomize Collision Normal Vector=false
    Radius Calculation Type=::NewEnumerator1
   )
   &Collision(
    Age Colliding Particles=false
    Analytical Collision Plane Position 1:Vector3f=<0 0 -20>
    Analytical Collision Plane Space:ENiagaraCoordinateSpace=::Local
    CPU Collision Type=::NewEnumerator3
    Collision Enabled=&IsPlatformSetActive(New Platform Set=DataInterface() {propertyValues {Platforms {qualityLevelMask=-28}}})
    KillOnCollision=false
    Module.Control Rest State Via Surface Normal=false
    Module.EnableMaxCollisionCount=false
    Module.Randomize Collision Normal Vector=false
    Perform 2 Analytical Collision Queries=false
    Restitution=0
   ) {enabled=false, module=Collision001}
   &PointForce(
    Force Strength=&FloatFromCurve(FloatCurve=DataInterface() {propertyValues {Curve {keys [<at 0 0>,<at 1 1>]}}}, CurveIndex=$Emitter.NormalizedLoopAge, Scale Curve=400)
    Module.Use Falloff=true
    Module.UseLinearForce=false
   )
   &Drag(
    Drag=&FloatFromCurve(
     FloatCurve=DataInterface() {propertyValues {Curve {keys [
         {interpMode=RCIM_Cubic, tangentMode=RCTM_User, time=0.388, value=0}
         {arriveTangent=4.8344, interpMode=RCIM_Cubic, leaveTangent=4.8344, tangentMode=RCTM_User, time=1, value=1}
        ]}}}
     Scale Curve=3
    )
    Use Linear Drag=true
    Use Rotational Drag=true
   )
   &SolveForcesAndVelocity(Clamp Velocity=false, Limit Acceleration=false)
   &AlignSpriteToMeshOrientation(SetAlignment=true, SetFacing=true)
  ]
  rendererClasses [&NiagaraMeshRendererProperties]
  renderers [{rendererClass=&NiagaraMeshRendererProperties, rendererIndex=0}]
 }
 VBucks001 {
  emitterUpdateScript [
   &EmitterState(Life Cycle Mode=::NewEnumerator1, Loop Behavior=::NewEnumerator1, Loop Delay=0.62, Loop Duration=2, UseLoopDelay=false)
   &SpawnBurst_Instantaneous(Spawn Count=20, Use Spawn Probability=true, UseLoopCountLimit=false)
   &SpawnRate(SpawnRate=5, Use Spawn Probability=false) {enabled=false, module=SpawnRate001}
   &SpawnRate(SpawnRate=10, Use Spawn Probability=false) {enabled=false}
   &SpawnPerUnit(
    Module.Use Max Movement Threshold=true
    Module.Use Movement Tolerance=false
    Spawn Probability=0.6
    Spawn Spacing:NiagaraFloat=&RandomRangeFloat(Minimum=10, Maximum=20)
    Use Spawn Probability=true
   ) {enabled=false}
  ]
  particleSpawnScript [
   &InitializeParticle(
    Color=&SelectLinearColorFromArray(
     Array Sampling Mode=::NewEnumerator0
     Direct Array Index=?
     Color Selection Array=DataInterface() {propertyValues {ColorData [<rgba 0.9 0.81 0.225 1>, <rgba 1 0.45 0.5875 1>, <rgba 0.36 0.9 0.5535 1>, <rgba 0.8019 0.5 1 1>, <rgba 1 0.5127 0.3 1>]}}
     Random Seed=0
    )
    Lifetime Max=1.6
    Lifetime Mode=::NewEnumerator1
    Mesh Scale Mode=::NewEnumerator3
    Mesh Uniform Scale=0.85
    UsePositionOffset=false
    Write Material Random=true
   )
   &FN_SetSpawnGroup(Pattern=::NewEnumerator0)
   &SkeletalMeshLocation(
    Bone / Socket Index=$Particles.SpawnGroup
    MaskSpawnGroup=false
    Mesh Sampling Type=::NewEnumerator5
    Skeletal Mesh=$System.Skeletal Mesh
    Socket Sampling Mode=::NewEnumerator5
   )
   &ShapeLocation(MaskSpawnGroup=false, Override Seed=false, Shape Origin=$Particles.SkeletalMeshLocation.SampledPosition, Sphere Radius=5, UseOwnerScale=true)
   &InitialMeshOrientation(Use Rotation Vector=false)
   &AddVelocity(Velocity=&RandomRangeVector(Minimum=<-300 -400 0>, Maximum=<300 400 50>)) {enabled=false}
   &AddVelocity(
    Cone Angle=15
    Cone Axis=&SelectVectorFromArray(
     Direct Array Index=$Particles.SpawnGroup
     Vector Selection Array=DataInterface() {propertyValues {FloatData [<1 0 0>, <-1 -1 0>, <0 0 0>, <-0.5 1 0>]}}
    )
    Module.UseConeFalloff=false
    Module.UseVelDistribution=false
    Override Seed=false
    Velocity Mode=::NewEnumerator2
    Velocity Speed=&RandomRangeFloat(Minimum=50, Maximum=120)
   ) {module=AddVelocity001}
   &AddVelocity(Rotation Coordinate Space=::World, Velocity=&RandomRangeVector(Minimum=<0 0 80>, Maximum=<0 0 300>)) {module=AddVelocity002}
  ]
  particleUpdateScript [
   /SetVariables(Particles.Drag:NiagaraFloat=&FloatFromCurve(FloatCurve=DataInterface() {propertyValues {Curve {keys [{tangentMode=RCTM_User, time=0, value=0}, {tangentMode=RCTM_User, time=0.2, value=1}]}}})) {isSetParameters=true}
   &ParticleState()
   &ScaleMeshSize(Scale Factor=&VectorFromFloat(Value=&FloatFromCurve(
      FloatCurve=DataInterface() {propertyValues {Curve {keys [
          {arriveTangent=4.1785, interpMode=RCIM_Cubic, leaveTangent=8.7018, tangentMode=RCTM_Break, time=0, value=0}
          {arriveTangent=0.0224, interpMode=RCIM_Cubic, leaveTangent=0.0224, tangentMode=RCTM_Break, time=0.3, value=1}
         ]}}}
      CurveIndex=$Particles.Age
      Scale Curve=&FloatFromCurve(FloatCurve=DataInterface() {propertyValues {Curve {keys [
           {interpMode=RCIM_Cubic, tangentMode=RCTM_Break, time=0.7176, value=1}
           {arriveTangent=-9.1079, interpMode=RCIM_Cubic, tangentMode=RCTM_Break, time=1, value=0}
          ]}}})
     )))
   &GravityForce()
   &CurlNoiseForce(Noise Frequency=4.9092, Noise Strength=1143.17, Pan Noise=false) {module=CurlNoiseForce001}
   &AerodynamicDrag(
    Aerodynamic Drag:NiagaraFloat=$Particles.Drag
    Aerodynamic Rotational Drag:NiagaraFloat=$Particles.Drag
    Rotation Strength:NiagaraFloat=&RandomRangeFloat(Minimum=2.4141, Maximum=5.2481)
    Use Linear Drag=true
    Use Rotational Drag=true
   )
   &Collision(
    Age Colliding Particles=false
    Collision Enabled=&IsPlatformSetActive(New Platform Set=DataInterface() {propertyValues {Platforms {qualityLevelMask=-8}}})
    KillOnCollision=false
    Module.Control Rest State Via Surface Normal=false
    Module.EnableMaxCollisionCount=false
    Module.Randomize Collision Normal Vector=false
    Radius Calculation Type=::NewEnumerator1
   )
   &Collision(
    Age Colliding Particles=false
    Analytical Collision Plane Position 1:Vector3f=<0 0 -20>
    Analytical Collision Plane Space:ENiagaraCoordinateSpace=::Local
    CPU Collision Type=::NewEnumerator3
    Collision Enabled=&IsPlatformSetActive(New Platform Set=DataInterface() {propertyValues {Platforms {qualityLevelMask=-28}}})
    KillOnCollision=false
    Module.Control Rest State Via Surface Normal=false
    Module.EnableMaxCollisionCount=false
    Module.Randomize Collision Normal Vector=false
    Perform 2 Analytical Collision Queries=false
    Restitution=0
   ) {enabled=false, module=Collision001}
   &Drag(
    Drag=&FloatFromCurve(
     FloatCurve=DataInterface() {propertyValues {Curve {keys [
         {interpMode=RCIM_Cubic, tangentMode=RCTM_User, time=0.388, value=0}
         {arriveTangent=4.8344, interpMode=RCIM_Cubic, leaveTangent=4.8344, tangentMode=RCTM_User, time=1, value=1}
        ]}}}
     Scale Curve=3
    )
    Use Linear Drag=true
    Use Rotational Drag=true
   )
   &PointForce(Force Strength=500, Module.Use Falloff=true, Module.UseLinearForce=false)
   &SolveForcesAndVelocity(Clamp Velocity=false, Limit Acceleration=false)
   &AlignSpriteToMeshOrientation(SetAlignment=true, SetFacing=true)
  ]
  rendererClasses [&NiagaraMeshRendererProperties]
  renderers [{rendererClass=&NiagaraMeshRendererProperties, rendererIndex=0}]
 }
}
summary {
 systemName=NS_LootFlex_MoneysNewOPPass
 emitters [
  {emitterName=VBucks, rendererClasses [&NiagaraMeshRendererProperties]}
  {emitterName=Coin, rendererClasses [&NiagaraMeshRendererProperties, &NiagaraSpriteRendererProperties]}
  {emitterName=VBucks001, rendererClasses [&NiagaraMeshRendererProperties]}
  {emitterName=Coin001, rendererClasses [&NiagaraSpriteRendererProperties, &NiagaraMeshRendererProperties]}
 ]
 userVariables [
  User.bIsFrontEnd {type=NiagaraBool}
  User.bIsFrontEndPreview {type=NiagaraBool}
 ]
}
